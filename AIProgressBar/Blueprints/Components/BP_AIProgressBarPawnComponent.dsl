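; BP_AIProgressBarPawnComponent is a pawn component in the AI Progress Bar feature that tracks which challenge manager its owning pawn is currently associated with.
; The OnRep handler for CurrentChallengeManager calls the OnCurrentChallengeManagerChanged dispatcher so UI and other listeners can react to the replicated change.
; It also writes a shipping log line with the owner's player name to record challenge zone changes received from the server.
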(fn OnRep_CurrentChallengeManager
  (Default|CallOnCurrentChallengeManagerChanged) ; Let UI, etc know that the challenge manager changed
  ;[1 0.8 0] Client debug logging for enter/exit
  (Fortnite|FortShippingLog 0 (Utilities|String|Append (Fortnite|GetPlayerNameSafe (Components|GetOwner)) " recieved challenge zone change from server")))
